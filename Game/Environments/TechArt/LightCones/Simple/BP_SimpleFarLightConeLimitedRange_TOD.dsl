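; Blueprint actor for a simple distant light-cone effect with limited range and time-of-day variation, drawn on a light-ray cone mesh.
; The construction script creates a dynamic material instance and pushes the actor's brightness, opacity, falloff, depth, height and near-fade variables into it.
; Use TOD Color selects four period colors or one color, while per-period brightness is always set.
; Bottom opacity falloff is optional, and the event graph is empty.
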(event EventBeginPlay)

(event Collision|EventActorBeginOverlap
  (params (OtherActor :type Actor%20Object%20Reference)))

(event EventTick
  (params (DeltaSeconds :type Float%20%28single-precision%29)))

(fn ConstructionScript
  (bind _mid (Variables|Default|GetMID))
  (bind _mid_2 (Variables|Default|GetMID))
  (bind _returnvalue (Rendering|Material|CreateDynamicMaterialInstance (Variables|Default|GetSM_Cone_SimpleLightRay) 0 "/Game/Environments/TechArt/LightCones/Simple/M_SimpleFarLightConeLimitedRange_TOD.M_SimpleFarLightConeLimitedRange_TOD" "None" :!class /Script/Engine.PrimitiveComponent))
  (bind _output_get (Variables|Default|SetMID _returnvalue))
  (Rendering|Material|SetScalarParameterValue _output_get "Brightness" (Variables|Default|GetBrightness) :!class /Script/Engine.MaterialInstanceDynamic)
  (if (Variables|TODM|GetUse%20TOD%20Color)
    (Rendering|Material|SetVectorParameterValue _mid_2 "Color Morning" (Variables|TODM|GetColor%20Morning) :!class /Script/Engine.MaterialInstanceDynamic)
    (Rendering|Material|SetVectorParameterValue _mid_2 "Color" (Variables|TODM|GetColor%20Day) :!class /Script/Engine.MaterialInstanceDynamic)
    (Rendering|Material|SetVectorParameterValue _mid_2 "Color Evening" (Variables|TODM|GetColor%20Evening) :!class /Script/Engine.MaterialInstanceDynamic)
    (Rendering|Material|SetVectorParameterValue _mid_2 "Color Night" (Variables|TODM|GetColor%20Night) :!class /Script/Engine.MaterialInstanceDynamic)
    (else
      (Rendering|Material|SetVectorParameterValue _mid_2 "Color" (Variables|Default|GetColor) :!class /Script/Engine.MaterialInstanceDynamic)))
  (Rendering|Material|SetScalarParameterValue (Variables|Default|GetMID) "DepthFade" (Variables|Default|GetDepthFade) :!class /Script/Engine.MaterialInstanceDynamic)
  (Rendering|Material|SetScalarParameterValue (Variables|Default|GetMID) "Opacity" (Variables|Default|GetOpacity) :!class /Script/Engine.MaterialInstanceDynamic)
  (Rendering|Material|SetScalarParameterValue (Variables|Default|GetMID) "Falloff" (Variables|Default|GetFalloff) :!class /Script/Engine.MaterialInstanceDynamic)
  (Rendering|Material|SetScalarParameterValue (Variables|Default|GetMID) "Falloff_Edges" (Variables|Default|GetFalloff_Edges) :!class /Script/Engine.MaterialInstanceDynamic)
  (Rendering|Material|SetScalarParameterValue (Variables|Default|GetMID) "HeightFadeOffset" (Variables|Default|GetHeightFadeOffset) :!class /Script/Engine.MaterialInstanceDynamic)
  (Rendering|Material|SetScalarParameterValue (Variables|Default|GetMID) "HeightFade_Blend" (Variables|Default|GetHeightFade_Blend) :!class /Script/Engine.MaterialInstanceDynamic)
  (Rendering|Material|SetScalarParameterValue (Variables|Default|GetMID) "NearFadeOffset" (Variables|Default|GetNearFadeOffset) :!class /Script/Engine.MaterialInstanceDynamic)
  (Rendering|Material|SetScalarParameterValue (Variables|Default|GetMID) "NearFadeBlend" (Variables|Default|GetNearFadeBlend) :!class /Script/Engine.MaterialInstanceDynamic)
  (Rendering|Material|SetScalarParameterValue (Variables|Default|GetMID) "Morning" (Variables|TODM|GetBrightness%20Morning) :!class /Script/Engine.MaterialInstanceDynamic)
  (Rendering|Material|SetScalarParameterValue (Variables|Default|GetMID) "Day" (Variables|TODM|GetBrightness%20Day) :!class /Script/Engine.MaterialInstanceDynamic)
  (Rendering|Material|SetScalarParameterValue (Variables|Default|GetMID) "Evening" (Variables|TODM|GetBrightness%20Evening) :!class /Script/Engine.MaterialInstanceDynamic)
  (Rendering|Material|SetScalarParameterValue _mid "Night" (Variables|TODM|GetBrightness%20Night) :!class /Script/Engine.MaterialInstanceDynamic)
  (if (Variables|Default|GetUseBottomOpacityFalloff)
    (Rendering|Material|SetScalarParameterValue _mid "BottomFalloff" (Variables|Default|GetBottomOpacityFalloff) :!class /Script/Engine.MaterialInstanceDynamic)
    (Rendering|Material|SetScalarParameterValue _mid "BottomOffset" (Variables|Default|GetBottom%20Offset) :!class /Script/Engine.MaterialInstanceDynamic)))
