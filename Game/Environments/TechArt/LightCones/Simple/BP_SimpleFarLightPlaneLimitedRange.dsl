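; Blueprint for a simple far light plane with limited range, a mesh-and-material light ray effect with per-time-of-day brightness.
; The construction script creates a dynamic instance of M_SimpleFarLightPlane_LimitedRange on the SM_Cone_SimpleLightRay component.
; It copies variables for color, brightness, opacity, falloff, depth, height and near fades, plus TODM Morning/Day/Evening/Night brightness, into material parameters.
; BeginPlay, overlap and Tick events are empty, so the Blueprint graph runs no gameplay logic.
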(event EventBeginPlay)

(event Collision|EventActorBeginOverlap
  (params (OtherActor :type Actor%20Object%20Reference)))

(event EventTick
  (params (DeltaSeconds :type Float%20%28single-precision%29)))

(fn ConstructionScript
  (bind _returnvalue (Rendering|Material|CreateDynamicMaterialInstance (Variables|Default|GetSM_Cone_SimpleLightRay) 0 "/Game/Environments/TechArt/LightCones/Simple/M_SimpleFarLightPlane_LimitedRange.M_SimpleFarLightPlane_LimitedRange" "None" :!class /Script/Engine.PrimitiveComponent))
  (bind _output_get (Variables|Default|SetMID _returnvalue))
  (Rendering|Material|SetScalarParameterValue _output_get "Brightness" (Variables|Default|GetBrightness) :!class /Script/Engine.MaterialInstanceDynamic)
  (Rendering|Material|SetVectorParameterValue (Variables|Default|GetMID) "Color" (Variables|Default|GetColor) :!class /Script/Engine.MaterialInstanceDynamic)
  (Rendering|Material|SetScalarParameterValue (Variables|Default|GetMID) "DepthFade" (Variables|Default|GetDepthFade) :!class /Script/Engine.MaterialInstanceDynamic)
  (Rendering|Material|SetScalarParameterValue (Variables|Default|GetMID) "Opacity" (Variables|Default|GetOpacity) :!class /Script/Engine.MaterialInstanceDynamic)
  (Rendering|Material|SetScalarParameterValue (Variables|Default|GetMID) "Falloff" (Variables|Default|GetFalloff) :!class /Script/Engine.MaterialInstanceDynamic)
  (Rendering|Material|SetScalarParameterValue (Variables|Default|GetMID) "Falloff_Edges" (Variables|Default|GetFalloff_Edges) :!class /Script/Engine.MaterialInstanceDynamic)
  (Rendering|Material|SetScalarParameterValue (Variables|Default|GetMID) "HeightFadeOffset" (Variables|Default|GetHeightFadeOffset) :!class /Script/Engine.MaterialInstanceDynamic)
  (Rendering|Material|SetScalarParameterValue (Variables|Default|GetMID) "HeightFade_Blend" (Variables|Default|GetHeightFade_Blend) :!class /Script/Engine.MaterialInstanceDynamic)
  (Rendering|Material|SetScalarParameterValue (Variables|Default|GetMID) "NearFadeOffset" (Variables|Default|GetNearFadeOffset) :!class /Script/Engine.MaterialInstanceDynamic)
  (Rendering|Material|SetScalarParameterValue (Variables|Default|GetMID) "NearFadeBlend" (Variables|Default|GetNearFadeBlend) :!class /Script/Engine.MaterialInstanceDynamic)
  (Rendering|Material|SetScalarParameterValue (Variables|Default|GetMID) "Morning" (Variables|TODM|GetBrightness%20Morning) :!class /Script/Engine.MaterialInstanceDynamic)
  (Rendering|Material|SetScalarParameterValue (Variables|Default|GetMID) "Day" (Variables|TODM|GetBrightness%20Day) :!class /Script/Engine.MaterialInstanceDynamic)
  (Rendering|Material|SetScalarParameterValue (Variables|Default|GetMID) "Evening" (Variables|TODM|GetBrightness%20Evening) :!class /Script/Engine.MaterialInstanceDynamic)
  (Rendering|Material|SetScalarParameterValue (Variables|Default|GetMID) "Night" (Variables|TODM|GetBrightness%20Night) :!class /Script/Engine.MaterialInstanceDynamic))
