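# Unlit additive material for instanced light cones seen from afar, with opacity scaled by time of day.
# A camera-to-cone distance fade hides nearby cones, combined with a camera height fade.
# Edge softness comes from a normal rebuilt from instance origin and axis against the view vector, multiplied by UV length falloffs and depth fade.
# When the combined fade reaches zero, an offset branch apparently collapses vertices to the origin, culling invisible cones.
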Material /Game/Environments/TechArt/LightCones/Simple/M_SimpleFarLightConeLimitedRange
graph {
 properties {
  bAutomaticallySetUsageInEditor=false
  bUsedWithInstancedStaticMeshes=true
  blendMode=BLEND_Additive
  pixelDepthOffsetMode=PDOM_Legacy
  shadingModel=MSM_Unlit
  usedShadingModels=Unlit
 }
 nodes [
  n1 = CameraPositionWS()
  n2 = NamedRerouteUsage()
  n3 = Divide(
   A=Subtract(A=Distance(A=@n1, B=@n2), B=ScalarParameter() {defaultValue=2000, parameterName=NearFadeOffset})
   B=ScalarParameter() {defaultValue=10000, parameterName=NearFadeBlend}
  )
  n4 = ComponentMask(None=@n1) {b=true}
  n5 = ComponentMask(None=@n2) {b=true}
  n6 = Divide(
   A=Subtract(A=Subtract(A=@n4, B=@n5), B=ScalarParameter() {defaultValue=-90000, parameterName=HeightFadeOffset})
   B=ScalarParameter() {defaultValue=5000, parameterName=HeightFade_Blend}
  )
  n7 = Multiply(
   A=MaterialFunctionCall(
    Value_Morning=ScalarParameter() {defaultValue=1, parameterName=Morning}
    Value_Day=ScalarParameter() {defaultValue=1, parameterName=Day}
    Value_Evening=ScalarParameter() {defaultValue=1, parameterName=Evening}
    Value_Night=ScalarParameter() {defaultValue=1, parameterName=Night}
   ) {materialFunction=&Scalar_Time_of_Day}.Result
   B=Min(A=Saturate(None=@n3), B=Saturate(None=@n6))
  )
  n8 = Subtract(A=NamedRerouteUsage(), B=WorldPosition().XYZ)
  If(A=@n7, "A > B"=Constant3Vector(), "A == B"=@n8, "A < B"=@n8)
  n9 = CameraVectorWS()
  n10 = TransformPosition(None=Constant3Vector()) {transformSourceType=TRANSFORMPOSSOURCE_Instance, transformType=TRANSFORMPOSSOURCE_World}
  n11 = Subtract(A=WorldPosition().XYZ, B=NamedRerouteDeclaration(None=@n10) {name=OriginPoint, nodeColor=<rgba 1 0.4453 0 1>})
  n12 = NamedRerouteUsage()
  n13 = Normalize(VectorInput=CrossProduct(A=CrossProduct(A=@n11, B=@n12), B=NamedRerouteUsage()))
  n14 = Power(Base=Multiply(A=DotProduct(A=@n9, B=@n13)) {constB=-1}, Exp=ScalarParameter() {defaultValue=2.5, parameterName=Falloff_Edges})
  n15 = ComponentMask(None=TextureCoordinate()) {g=true}
  n16 = Multiply(
   A=ScalarParameter() {defaultValue=1, parameterName=Opacity}
   B=Power(Base=@n15, Exp=ScalarParameter() {defaultValue=3, parameterName=Falloff})
  )
  n17 = Subtract(A=OneMinus(None=@n15), B=ScalarParameter() {defaultValue=0.08, parameterName=Offset})
  n18 = ScalarParameter() {defaultValue=8, parameterName=BottomFalloff}
  n19 = StaticSwitchParameter(True=Multiply(A=@n16, B=Multiply(A=@n17, B=@n18)), False=@n16) {defaultValue=true, parameterName=useBottomAreaFalloff}
  Multiply(A=DepthFade(Opacity=Multiply(A=@n14, B=@n19), FadeDistance=ScalarParameter() {defaultValue=200, parameterName=DepthFade}), B=@n7)
  Multiply(A=VectorParameter() {parameterName=Color, defaultValue=<rgba 1 1 1 1>}.RGB, B=ScalarParameter() {defaultValue=2, parameterName=Brightness})
  n20 = Transform(None=Constant3Vector() {constant=<rgba 0 0 1 0>}) {transformSourceType=TRANSFORMSOURCE_Instance}
  NamedRerouteDeclaration(None=@n20) {name=ConeDirection, nodeColor=<rgba 0.3828 0 1 1>}
  Reroute(None=@n20)
  Reroute(None=@n10)
 ]
}
