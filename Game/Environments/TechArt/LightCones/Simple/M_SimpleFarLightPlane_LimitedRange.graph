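# Additive unlit material for a simple far light plane, a soft colored glow whose visibility depends on time of day and camera distance.
# A noise-modulated radial mask with adjustable edge falloff passes through depth fade, then is multiplied by time-of-day and near-distance blend factors.
# When that combined factor reaches zero, an If yields an offset from world position to a rerouted point, apparently collapsing vertices to the instance origin.
# The graph also computes a camera height fade relative to that origin.
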Material /Game/Environments/TechArt/LightCones/Simple/M_SimpleFarLightPlane_LimitedRange
graph {
 properties {bAutomaticallySetUsageInEditor=false, blendMode=BLEND_Additive, shadingModel=MSM_Unlit, usedShadingModels=Unlit}
 nodes [
  Abs()
  DistanceCullFade()
  n1 = ScalarParameter() {defaultValue=10000, parameterName=NearFadeBlend}
  n2 = ScalarParameter() {defaultValue=2000, parameterName=NearFadeOffset}
  n3 = Multiply(
   A=MaterialFunctionCall(
    Value_Morning=ScalarParameter() {defaultValue=1, parameterName=Morning}
    Value_Day=ScalarParameter() {defaultValue=1, parameterName=Day}
    Value_Evening=ScalarParameter() {defaultValue=1, parameterName=Evening}
    Value_Night=ScalarParameter() {defaultValue=1, parameterName=Night}
   ) {materialFunction=&Scalar_Time_of_Day}.Result
   B=MaterialFunctionCall!(Blend Range=@n1, Start Offset=@n2) {materialFunction=/Engine/Functions/Engine_MaterialFunctions03/Particles/Distance_Blend}.Result
  )
  n4 = Subtract(A=NamedRerouteUsage(), B=WorldPosition().XYZ)
  If(A=@n3, "A > B"=Constant3Vector(), "A == B"=@n4, "A < B"=@n4)
  MaterialFunctionCall(
   Value_Morning=ScalarParameter() {defaultValue=1, parameterName=Morning}
   Value_Day=ScalarParameter() {defaultValue=1, parameterName=Day}
   Value_Evening=ScalarParameter() {defaultValue=1, parameterName=Evening}
   Value_Night=ScalarParameter() {defaultValue=1, parameterName=Night}
  ) {materialFunction=&Scalar_Time_of_Day}
  n5 = CameraPositionWS()
  n6 = NamedRerouteDeclaration(None=TransformPosition(None=Constant3Vector()) {transformSourceType=TRANSFORMPOSSOURCE_Instance, transformType=TRANSFORMPOSSOURCE_World}) {name=OriginPoint, nodeColor=<rgba 1 0.4453 0 1>}
  n7 = Subtract(A=Distance(A=@n5, B=@n6), B=@n2)
  n8 = Subtract(
   A=Subtract(A=ComponentMask(None=@n5) {b=true}, B=ComponentMask(None=@n6) {b=true})
   B=ScalarParameter() {defaultValue=-90000, parameterName=HeightFadeOffset}
  )
  n9 = ScalarParameter() {defaultValue=5000, parameterName=HeightFade_Blend}
  Min(A=Saturate(None=Divide(A=@n7, B=@n1)), B=Saturate(None=Divide(A=@n8, B=@n9)))
  n10 = ScalarParameter() {defaultValue=1, parameterName=Opacity}
  n11 = TextureCoordinate()
  Multiply(A=@n10, B=Power(Base=ComponentMask(None=@n11) {g=true}, Exp=ScalarParameter() {defaultValue=3, parameterName=Falloff}))
  n12 = TextureSample() {texture=&AtmosphericCloudNoise01}
  Multiply(A=@n12.R)
  n13 = TextureCoordinate()
  n14 = Constant2Vector() {g=0.5, r=0.5}
  n15 = Multiply(
   A=LinearInterpolate(Alpha=@n12.R) {constA=0.666, constB=1.15}
   B=Power(
    Base=SphereMask(A=@n13, B=@n14) {attenuationRadius=0.5, hardnessPercent=0}
    Exp=ScalarParameter() {defaultValue=2.5, parameterName=Falloff_Edges}
   )
  )
  Multiply(A=DepthFade(Opacity=Multiply(A=@n15, B=@n10), FadeDistance=ScalarParameter() {defaultValue=200, parameterName=DepthFade}), B=@n3)
  Multiply(A=VectorParameter() {parameterName=Color, defaultValue=<rgba 1 1 1 1>}.RGB, B=ScalarParameter() {defaultValue=2, parameterName=Brightness})
 ]
}
